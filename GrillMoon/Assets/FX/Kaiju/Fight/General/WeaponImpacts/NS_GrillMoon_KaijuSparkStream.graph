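# Niagara system for a continuous spark stream in kaiju fight weapon impacts, exposing two user positions, SparkDirA and SparkDirB.
# Matching GPU and CPU emitters spray dense, fast sparks in a narrow cone, while a third CPU emitter adds sparser, slower sparks with an age-based kill.
# All use sprites with randomized gravity, drag and an HDR white-hot-to-red fade, with System.FX Scale scaling velocity, gravity and size.
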NiagaraSystem /GrillMoon/Assets/FX/Kaiju/Fight/General/WeaponImpacts/NS_GrillMoon_KaijuSparkStream
dependencies {
 usedDataInterfaces [&NiagaraDataInterfaceCurve, &NiagaraDataInterfaceColorCurve]
 usedDynamicInputs [
  &self:Mask_Vector_by_Int_01
  &NormalizeVector
  &TransformVector
 ]
 usedModules [
  &self:SystemSpawnScript.SystemScriptSource.SystemScriptGraph.NiagaraNodeAssignment_0.SetVariables_#1
  &SystemState
 ]
}
emitters {
 SparksDir {
  simTarget=GPUComputeSim
  emitterUpdateScript [
   &EmitterState()
   &SpawnRate(SpawnRate=&Multiply_Float(A=2, B=1000), Use Spawn Probability=false)
  ]
  particleSpawnScript [
   &InitializeParticle(
    Color Mode=::NewEnumerator0
    Lifetime Mode=::NewEnumerator1
    Sprite Size Max=<5 50>
    Sprite Size Min=<15 25>
    Sprite Size Mode=::NewEnumerator2
    UsePositionOffset=false
    Write Material Random=true
   )
   &PartitionParticles()
   &AddVelocity(
    Cone Angle=5
    Module.UseConeFalloff=false
    Module.UseVelDistribution=false
    Override Seed=false
    Velocity Mode=::NewEnumerator2
    Velocity Speed=&RandomRangeFloat(Minimum=100, Maximum=3000)
   )
   &ShapeLocation(MaskSpawnGroup=false, Override Seed=false, Sphere Radius=150, UseOwnerScale=true)
   /SetVariables(Particles.Velocity=&Multiply_VectorByFloat(Vector=$Particles.Velocity, Float=$System.FX Scale)) {isSetParameters=true, module=SetVariables_#2}
  ]
  particleUpdateScript [
   &GravityForce(Gravity=&Multiply_VectorByFloat(
     Vector=<0 0 -980>
     Float=&Multiply_Float(A=&RandomRangeFloat(Minimum=0.25, Maximum=1), B=$System.FX Scale)
    ))
   &Drag(Drag=0.35, Use Linear Drag=true, Use Rotational Drag=true)
   &SolveForcesAndVelocity(Clamp Velocity=false, Limit Acceleration=false)
   &ParticleState()
   &ScaleSpriteSize(
    Uniform Curve Scale=$System.FX Scale
    Uniform Curve Sprite Scale=DataInterface() {propertyValues {Curve {keys [<at 0.8 1 RCIM_Cubic>, {arriveTangent=-15.439, interpMode=RCIM_Cubic, leaveTangent=-15.439, tangentMode=RCTM_User, time=1, value=0}]}}}
   )
   &Color(
    Color=&ColorFromCurve(ColorCurve=DataInterface() {propertyValues {
       AlphaCurve {keys [<at 0 1 RCIM_Cubic>,<at 1 0 RCIM_Cubic>]}
       BlueCurve {keys [
         <at 0 37.5>
         {arriveTangentWeight=3.3334, interpMode=RCIM_Cubic, leaveTangentWeight=21.667, tangentMode=RCTM_User, time=0.0737, value=0}
         <at 0.4455 0 RCIM_Cubic>
         <at 0.9911 0>
        ]}
       GreenCurve {keys [
         <at 0 46.875>
         {arriveTangent=-105.2195, interpMode=RCIM_Cubic, leaveTangent=-105.2195, time=0.0737, value=18.75}
         {arriveTangent=-20.4383, interpMode=RCIM_Cubic, leaveTangent=-20.4383, time=0.4455, value=0}
         <at 0.9911 0>
        ]}
       RedCurve {keys [
         <at 0 75>
         {arriveTangent=-228.9853, interpMode=RCIM_Cubic, leaveTangent=-228.9853, tangentMode=RCTM_User, time=0.0737, value=75}
         {arriveTangent=-81.4808, interpMode=RCIM_Cubic, leaveTangent=-81.4808, time=0.4455, value=25}
         <at 0.9911 0.25>
        ]}
      }})
    Scale Alpha=&FloatFromCurve(FloatCurve=DataInterface() {propertyValues {Curve {keys [<at 0 1 RCIM_Cubic>, {arriveTangent=-3.2506, interpMode=RCIM_Cubic, leaveTangent=-3.2506, tangentMode=RCTM_User, time=1, value=0}]}}})
    Scale Color=&VectorFromFloat(Value=0.15)
   )
  ]
  rendererClasses [&NiagaraSpriteRendererProperties]
  renderers [{rendererClass=&NiagaraSpriteRendererProperties}]
 }
 SparksDir001 {
  emitterUpdateScript [
   &EmitterState()
   &SpawnRate(SpawnRate=100, Use Spawn Probability=false)
  ]
  particleSpawnScript [
   &InitializeParticle(
    Color Mode=::NewEnumerator0
    Lifetime Max=1.5
    Lifetime Mode=::NewEnumerator1
    Sprite Size Max=<5 50>
    Sprite Size Min=<15 25>
    Sprite Size Mode=::NewEnumerator2
    UsePositionOffset=false
    Write Material Random=true
   )
   &PartitionParticles()
   &AddVelocity(
    Module.Use Falloff Distance=false
    Override Seed=false
    Velocity Mode=::NewEnumerator1
    Velocity Speed=&RandomRangeFloat(Minimum=50, Maximum=1200)
   )
   &ShapeLocation(MaskSpawnGroup=false, Override Seed=false, Sphere Radius=0, UseOwnerScale=true)
   /SetVariables(Particles.Velocity=&Multiply_VectorByFloat(Vector=$Particles.Velocity, Float=$System.FX Scale)) {isSetParameters=true, module=SetVariables_#2}
  ]
  particleUpdateScript [
   &GravityForce(Gravity=&Multiply_VectorByFloat(
     Vector=<0 0 -980>
     Float=&Multiply_Float(A=&RandomRangeFloat(Minimum=0.25, Maximum=1), B=$System.FX Scale)
    ))
   &Drag(Drag=0.35, Use Linear Drag=true, Use Rotational Drag=true)
   &SolveForcesAndVelocity(Clamp Velocity=false, Limit Acceleration=false)
   &ParticleState()
   &ScaleSpriteSize(
    Uniform Curve Scale=$System.FX Scale
    Uniform Curve Sprite Scale=DataInterface() {propertyValues {Curve {keys [<at 0.8 1 RCIM_Cubic>, {arriveTangent=-15.439, interpMode=RCIM_Cubic, leaveTangent=-15.439, tangentMode=RCTM_User, time=1, value=0}]}}}
   )
   &Color(
    Color=&ColorFromCurve(ColorCurve=DataInterface() {propertyValues {
       AlphaCurve {keys [<at 0 1 RCIM_Cubic>,<at 1 0 RCIM_Cubic>]}
       BlueCurve {keys [
         <at 0 37.5>
         {arriveTangentWeight=3.3334, interpMode=RCIM_Cubic, leaveTangentWeight=21.667, tangentMode=RCTM_User, time=0.0737, value=0}
         <at 0.4455 0 RCIM_Cubic>
         <at 0.9911 0>
        ]}
       GreenCurve {keys [
         <at 0 46.875>
         {arriveTangent=-105.2195, interpMode=RCIM_Cubic, leaveTangent=-105.2195, time=0.0737, value=18.75}
         {arriveTangent=-20.4383, interpMode=RCIM_Cubic, leaveTangent=-20.4383, time=0.4455, value=0}
         <at 0.9911 0>
        ]}
       RedCurve {keys [
         <at 0 75>
         {arriveTangent=-228.9853, interpMode=RCIM_Cubic, leaveTangent=-228.9853, tangentMode=RCTM_User, time=0.0737, value=75}
         {arriveTangent=-81.4808, interpMode=RCIM_Cubic, leaveTangent=-81.4808, time=0.4455, value=25}
         <at 0.9911 0.25>
        ]}
      }})
    Scale Alpha=&FloatFromCurve(FloatCurve=DataInterface() {propertyValues {Curve {keys [<at 0 1 RCIM_Cubic>, {arriveTangent=-3.2506, interpMode=RCIM_Cubic, leaveTangent=-3.2506, tangentMode=RCTM_User, time=1, value=0}]}}})
    Scale Color=&VectorFromFloat(Value=0.15)
   )
   &KillParticles(Kill Particles=&SetBoolByFloatComparison(A=$Particles.NormalizedAge, B=0.65))
  ]
  rendererClasses [&NiagaraSpriteRendererProperties]
  renderers [{rendererClass=&NiagaraSpriteRendererProperties}]
 }
 SparksDirCpu {
  emitterUpdateScript [
   &EmitterState()
   &SpawnRate(SpawnRate=&Multiply_Float(A=2, B=1000), Use Spawn Probability=false)
  ]
  particleSpawnScript [
   &InitializeParticle(
    Color Mode=::NewEnumerator0
    Lifetime Mode=::NewEnumerator1
    Sprite Size Max=<5 50>
    Sprite Size Min=<15 25>
    Sprite Size Mode=::NewEnumerator2
    UsePositionOffset=false
    Write Material Random=true
   )
   &PartitionParticles()
   &AddVelocity(
    Cone Angle=5
    Module.UseConeFalloff=false
    Module.UseVelDistribution=false
    Override Seed=false
    Velocity Mode=::NewEnumerator2
    Velocity Speed=&RandomRangeFloat(Minimum=100, Maximum=3000)
   )
   &ShapeLocation(MaskSpawnGroup=false, Override Seed=false, Sphere Radius=150, UseOwnerScale=true)
   /SetVariables(Particles.Velocity=&Multiply_VectorByFloat(Vector=$Particles.Velocity, Float=$System.FX Scale)) {isSetParameters=true, module=SetVariables_#2}
  ]
  particleUpdateScript [
   &GravityForce(Gravity=&Multiply_VectorByFloat(
     Vector=<0 0 -980>
     Float=&Multiply_Float(A=&RandomRangeFloat(Minimum=0.25, Maximum=1), B=$System.FX Scale)
    ))
   &Drag(Drag=0.35, Use Linear Drag=true, Use Rotational Drag=true)
   &SolveForcesAndVelocity(Clamp Velocity=false, Limit Acceleration=false)
   &ParticleState()
   &ScaleSpriteSize(
    Uniform Curve Scale=$System.FX Scale
    Uniform Curve Sprite Scale=DataInterface() {propertyValues {Curve {keys [<at 0.8 1 RCIM_Cubic>, {arriveTangent=-15.439, interpMode=RCIM_Cubic, leaveTangent=-15.439, tangentMode=RCTM_User, time=1, value=0}]}}}
   )
   &Color(
    Color=&ColorFromCurve(ColorCurve=DataInterface() {propertyValues {
       AlphaCurve {keys [<at 0 1 RCIM_Cubic>,<at 1 0 RCIM_Cubic>]}
       BlueCurve {keys [
         <at 0 37.5>
         {arriveTangentWeight=3.3334, interpMode=RCIM_Cubic, leaveTangentWeight=21.667, tangentMode=RCTM_User, time=0.0737, value=0}
         <at 0.4455 0 RCIM_Cubic>
         <at 0.9911 0>
        ]}
       GreenCurve {keys [
         <at 0 46.875>
         {arriveTangent=-105.2195, interpMode=RCIM_Cubic, leaveTangent=-105.2195, time=0.0737, value=18.75}
         {arriveTangent=-20.4383, interpMode=RCIM_Cubic, leaveTangent=-20.4383, time=0.4455, value=0}
         <at 0.9911 0>
        ]}
       RedCurve {keys [
         <at 0 75>
         {arriveTangent=-228.9853, interpMode=RCIM_Cubic, leaveTangent=-228.9853, tangentMode=RCTM_User, time=0.0737, value=75}
         {arriveTangent=-81.4808, interpMode=RCIM_Cubic, leaveTangent=-81.4808, time=0.4455, value=25}
         <at 0.9911 0.25>
        ]}
      }})
    Scale Alpha=&FloatFromCurve(FloatCurve=DataInterface() {propertyValues {Curve {keys [<at 0 1 RCIM_Cubic>, {arriveTangent=-3.2506, interpMode=RCIM_Cubic, leaveTangent=-3.2506, tangentMode=RCTM_User, time=1, value=0}]}}})
    Scale Color=&VectorFromFloat(Value=0.15)
   )
  ]
  rendererClasses [&NiagaraSpriteRendererProperties]
  renderers [{rendererClass=&NiagaraSpriteRendererProperties}]
 }
}
summary {
 systemName=NS_GrillMoon_KaijuSparkStream
 emitters [
  {emitterName=SparksDir, simTarget=GPUComputeSim, rendererClasses [&NiagaraSpriteRendererProperties]}
  {emitterName=SparksDir001, rendererClasses [&NiagaraSpriteRendererProperties]}
  {emitterName=SparksDirCpu, rendererClasses [&NiagaraSpriteRendererProperties]}
 ]
 userVariables [
  User.SparkDirA {type=NiagaraPosition, description="", defaultValue {}}
  User.SparkDirB {type=NiagaraPosition, description="", defaultValue {}}
 ]
}
