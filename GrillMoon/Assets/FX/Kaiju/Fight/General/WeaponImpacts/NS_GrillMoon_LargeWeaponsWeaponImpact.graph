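# Niagara impact effect for large weapon hits in GrillMoon kaiju fights, scaled largely through System FX Scale.
# Flash layers sprites, a light and an expanding mesh with animated refraction amount, while SoftAirflow bursts rotating, fading meshes.
# SparksDir (GPU) and SparksDir002 (CPU) fire narrow spark cones along User.SparkDirA or SparkDirB, chosen by particle partition.
# SparksDir001 adds non-directed sparks; all sparks use gravity, drag and hot-to-red color fading.
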NiagaraSystem /GrillMoon/Assets/FX/Kaiju/Fight/General/WeaponImpacts/NS_GrillMoon_LargeWeaponsWeaponImpact
dependencies {
 usedDataInterfaces [&NiagaraDataInterfaceCurve, &NiagaraDataInterfaceColorCurve, &NiagaraDataInterfaceDebugDraw]
 usedDynamicInputs []
 usedModules [
  &self:SystemSpawnScript.SystemScriptSource.SystemScriptGraph.NiagaraNodeAssignment_4.SetVariables_#1
  &SystemState
  &self:ScratchModule
 ]
}
emitters {
 Flash {
  emitterSpawnScript [/SetVariables(Emitter.MeshSpeedMul=1.33) {isSetParameters=true}]
  emitterUpdateScript [
   &EmitterState()
   &SpawnBurst_Instantaneous(Use Spawn Probability=false, UseLoopCountLimit=false)
  ]
  particleSpawnScript [&InitializeParticle(
    Color=<rgba 1 0.9027 0.639 1>
    Lifetime=1.5
    Mesh Scale Mode=::NewEnumerator3
    Mesh Uniform Scale=25
    Sprite Rotation Mode=::NewEnumerator0
    Sprite Size Mode=::NewEnumerator3
    Uniform Sprite Size=&Multiply_Float(A=2, B=1000)
    UsePositionOffset=false
    Write Material Random=true
   )]
  particleUpdateScript [
   &ParticleState()
   &ScaleMeshSize(Scale Factor=&VectorFromFloat(Value=&FloatFromCurve(
      FloatCurve=DataInterface() {propertyValues {Curve {keys [
          {arriveTangent=0.2746, interpMode=RCIM_Cubic, leaveTangent=0.2746, tangentMode=RCTM_User, time=0, value=0}
          {arriveTangent=1.3466, tangentMode=RCTM_User, time=1, value=1}
         ]}}}
      CurveIndex=&Multiply_Float(A=$Particles.NormalizedAge, B=$Emitter.MeshSpeedMul)
      Scale Curve=&Multiply_Float(A=1, B=$System.FX Scale)
     )))
   &ScaleSpriteSize(
    Uniform Curve Scale=$System.FX Scale
    Uniform Curve Sprite Scale=DataInterface() {propertyValues {Curve {keys [
        {
         arriveTangent=2.1284
         arriveTangentWeight=0.4603
         interpMode=RCIM_Cubic
         leaveTangent=2.1284
         leaveTangentWeight=0.4603
         tangentMode=RCTM_User
         tangentWeightMode=RCTWM_WeightedBoth
         time=0
         value=0
        }
        {
         arriveTangentWeight=0.6435
         interpMode=RCIM_Cubic
         leaveTangentWeight=0.6435
         tangentMode=RCTM_User
         tangentWeightMode=RCTWM_WeightedBoth
         time=1
         value=1
        }
       ]}}}
   )
   &ScaleColor(
    Scale Alpha=&FloatFromCurve(
     FloatCurve=DataInterface() {propertyValues {Curve {keys [
         {arriveTangent=-36.5506, interpMode=RCIM_Cubic, leaveTangent=-36.5506, tangentMode=RCTM_User, time=0, value=5}
         {arriveTangent=-2.1686, interpMode=RCIM_Cubic, leaveTangent=-2.1686, tangentMode=RCTM_User, time=0.4003, value=0.9}
         {interpMode=RCIM_Cubic, leaveTangent=-1.8978, time=1, value=0}
        ]}}}
     Scale Curve=0.2
    )
    Scale RGB=&VectorFromFloat(Value=5)
    ScaleA=true
    ScaleRGB=true
   )
   &FN_Light_Attributes(
    Color=$Particles.Color
    Create Color Attribute=true
    Create Exponent Attribute=true
    Create Radius Attribute=true
    Exponent=1.5
    Radius=&FloatFromCurve(
     FloatCurve=DataInterface() {propertyValues {Curve {keys [
         {arriveTangent=9.0725, interpMode=RCIM_Cubic, leaveTangent=9.0725, tangentMode=RCTM_User, time=0, value=0}
         {arriveTangent=-2.5249, interpMode=RCIM_Cubic, leaveTangent=-2.5249, tangentMode=RCTM_User, time=0.1, value=0.9847}
        ]}}}
     Scale Curve=24472.5293
    )
    Scale Color=6
   )
   &DynamicMaterialParameters(
    Index 0 Param 1=&Lerp_Float(
     A=2
     B=0.5
     Alpha=&FloatFromCurve(FloatCurve=DataInterface() {propertyValues {Curve {keys [
          {arriveTangent=1.8767, interpMode=RCIM_Cubic, leaveTangent=1.8767, tangentMode=RCTM_User, time=0, value=0}
          {tangentMode=RCTM_User, time=1, value=1}
         ]}}})
    )
    Index 0 Param 2=1
    Index 0 Param 3=1
    Index 0 Param 4=1
    Index 1 Param 1=0.6
    Index 1 Param 2=1
    Index 1 Param 3=1
    Index 1 Param 4=1
    Module.Index0Param0Write=true
    Module.Index0Param1Write=true
    Module.Index0Param2Write=true
    Module.Index0Param3Write=true
    Module.Index1Param0Write=true
    Module.Index1Param1Write=true
    Module.Index1Param2Write=true
    Module.Index1Param3Write=true
    Write Parameter Index 1=true
   )
   /SetVariables(Particles.SPrite 2 dynam0:Vector4f=&MakeVector4(X=1, Y=500, Z=0, W=0)) {isSetParameters=true}
   /SetVariables(
    Particles.Sprite 3 Color:LinearColor=&MultiplyLinearColor(
     Linear Color=$Particles.Color
     Scale Factor=&MakeFloat4FromVectorAndFloat("Vector (RGB)"=&VectorFromFloat(Value=2), "Float (Alpha)"=1)
    )
    Particles.Sprite 3 size:Vector2f=&Multiply_Vector2D(A=$Particles.SpriteSize, B=<3 0.1>)
   ) {isSetParameters=true}
   /SetVariables(Particles.MeshReffract amount:NiagaraFloat=&FloatFromCurve(
     FloatCurve=DataInterface() {propertyValues {Curve {keys [<at 0.5 1 RCIM_Cubic>,<at 1 0>]}}}
     CurveIndex=&Multiply_Float(A=$Particles.NormalizedAge, B=$Emitter.MeshSpeedMul)
     Scale Curve=0.35
    )) {isSetParameters=true}
   /SetVariables(
    Particles.Flare Size:Vector2f=&Multiply_Vector2D(
     A=$Particles.Initial.SpriteSize
     B=&Vector2DFromFloat(Value=&FloatFromCurve(
       FloatCurve=DataInterface() {propertyValues {Curve {keys [
           {
            arriveTangent=2.1284
            arriveTangentWeight=0.4603
            interpMode=RCIM_Cubic
            leaveTangent=2.1284
            leaveTangentWeight=0.4603
            tangentMode=RCTM_User
            tangentWeightMode=RCTWM_WeightedBoth
            time=0
            value=0
           }
           {
            arriveTangentWeight=0.6435
            interpMode=RCIM_Cubic
            leaveTangentWeight=0.6435
            tangentMode=RCTM_User
            tangentWeightMode=RCTWM_WeightedBoth
            time=1
            value=1
           }
          ]}}}
       CurveIndex=&Multiply_Float(A=$Particles.NormalizedAge, B=2)
       Scale Curve=$System.FX Scale
      ))
    )
    Particles.flare Color:LinearColor=&MultiplyLinearColor(
     Linear Color=$Particles.Initial.Color
     Scale Factor=&MakeFloat4FromVectorAndFloat(
      "Vector (RGB)"=&VectorFromFloat(Value=5)
      "Float (Alpha)"=&FloatFromCurve(
       FloatCurve=DataInterface() {propertyValues {Curve {keys [
           {arriveTangent=-36.5506, interpMode=RCIM_Cubic, leaveTangent=-36.5506, tangentMode=RCTM_User, time=0, value=5}
           {arriveTangent=-2.1686, interpMode=RCIM_Cubic, leaveTangent=-2.1686, tangentMode=RCTM_User, time=0.4003, value=0.9}
           {interpMode=RCIM_Cubic, leaveTangent=-1.8978, time=1, value=0}
          ]}}}
       CurveIndex=&Multiply_Float(A=$Particles.NormalizedAge, B=2)
       Scale Curve=0.2
      )
     )
    )
    Particles.flaredynam 9:Vector4f=&MakeVector4(
     X=&Lerp_Float(
      A=2
      B=0.5
      Alpha=&FloatFromCurve(
       FloatCurve=DataInterface() {propertyValues {Curve {keys [
           {arriveTangent=1.8767, interpMode=RCIM_Cubic, leaveTangent=1.8767, tangentMode=RCTM_User, time=0, value=0}
           {tangentMode=RCTM_User, time=1, value=1}
          ]}}}
       Scale Curve=2
      )
     )
     Y=1
     Z=1
     W=1
    )
   ) {isSetParameters=true}
  ]
  rendererClasses [
   &NiagaraSpriteRendererProperties
   &NiagaraLightRendererProperties
   &NiagaraMeshRendererProperties
  ]
  renderers [
   {rendererClass=&NiagaraSpriteRendererProperties}
   {rendererClass=&NiagaraSpriteRendererProperties, rendererIndex=1}
   {rendererClass=&NiagaraLightRendererProperties, rendererIndex=2}
   {rendererClass=&NiagaraSpriteRendererProperties, rendererIndex=3}
   {rendererClass=&NiagaraMeshRendererProperties, rendererIndex=4}
  ]
 }
 SoftAirflow {
  emitterUpdateScript [
   &EmitterState()
   &SpawnBurst_Instantaneous(Spawn Count=5, Use Spawn Probability=false, UseLoopCountLimit=false)
  ]
  particleSpawnScript [
   &InitializeParticle(
    Color=<rgba 0.4949 0.6699 0.8698 1>
    Lifetime Max=3
    Lifetime Min=1.5
    Lifetime Mode=::NewEnumerator1
    Mesh Scale=&Multiply_VectorByFloat(
     Vector=&VectorFromFloat(Value=$System.FX Scale)
     Float=&RandomRangeFloat(Minimum=15, Maximum=10)
    )
    Mesh Scale Mode=::NewEnumerator4
    Position Offset=<0 0 200>
    UsePositionOffset=false
    Write Material Random=true
   )
   &InitialMeshOrientation(
    Mesh Orientation Mode=::NewEnumerator7
    Rotation=&Multiply_Vector(A=&RandomRangeVector(Minimum=<-1 -1 -1>, Maximum=<1 1 1>), B=<0.05 0.05 1>)
    Use Rotation Vector=true
   )
   &ShapeLocation(
    Cylinder Height=70
    Cylinder Radius=0
    MaskSpawnGroup=false
    Override Seed=false
    Shape Primitive=::NewEnumerator1
    Surface Only=false
    UseOwnerScale=true
   )
  ]
  particleUpdateScript [
   &ParticleState()
   &DynamicMaterialParameters(
    Index 0 Param 1=&FloatFromCurve(FloatCurve=DataInterface() {propertyValues {Curve {keys [<at 0.5 0>, {interpMode=RCIM_Cubic, tangentMode=RCTM_User, time=1, value=1}]}}})
    Index 0 Param 3=1
    Module.Param0WriteEnabled=true
    Module.Param1WriteEnabled=true
    Module.Param2WriteEnabled=true
    Module.Param3WriteEnabled=true
   )
   &ScaleMeshSize(Scale Factor=&Multiply_VectorByFloat(
     Vector=<1 1 1.5>
     Float=&FloatFromCurve(
      FloatCurve=DataInterface() {propertyValues {Curve {keys [
          {arriveTangent=2.376, time=0, value=0}
          {arriveTangent=0.1286, interpMode=RCIM_Cubic, leaveTangent=0.1286, tangentMode=RCTM_User, time=1, value=1}
         ]}}}
      Scale Curve=2
     )
    ))
   &UpdateMeshOrientation(Rotation Rate=0.15, Rotation Vector=<0 0 1>)
   &ScaleColor(
    Scale Alpha=&FloatFromCurve(FloatCurve=DataInterface() {propertyValues {Curve {keys [
         {
          arriveTangent=9.292
          arriveTangentWeight=0.4785
          interpMode=RCIM_Cubic
          leaveTangent=9.292
          leaveTangentWeight=0.4785
          tangentMode=RCTM_User
          tangentWeightMode=RCTWM_WeightedBoth
          time=0
          value=0
         }
         {arriveTangent=-1.9397, interpMode=RCIM_Cubic, leaveTangent=-2.9533, tangentMode=RCTM_Break, time=0.5, value=0.5}
         {arriveTangent=-0.7636, interpMode=RCIM_Cubic, leaveTangent=-0.7636, tangentMode=RCTM_User, time=1, value=0}
        ]}}})
    ScaleA=true
    ScaleRGB=true
   )
  ]
  rendererClasses [&NiagaraMeshRendererProperties]
  renderers [{rendererClass=&NiagaraMeshRendererProperties, rendererIndex=0}]
 }
 SparksDir {
  simTarget=GPUComputeSim
  emitterUpdateScript [
   &EmitterState()
   &SpawnRate(SpawnRate=10000, Use Spawn Probability=false)
  ]
  particleSpawnScript [
   &InitializeParticle(
    Color Mode=::NewEnumerator0
    Lifetime Max=2.5
    Lifetime Mode=::NewEnumerator1
    Sprite Size Max=<5 50>
    Sprite Size Min=<15 25>
    Sprite Size Mode=::NewEnumerator2
    UsePositionOffset=false
    Write Material Random=true
   )
   &PartitionParticles()
   &AddVelocity(
    Cone Angle=5
    Cone Axis=&self:Mask_Vector_by_Int(
     Int Mask:NiagaraInt32=0
     int:NiagaraInt32=$Particles.PartitionParticles.Partition
     Masked Vector:Vector3f=&NormalizeVector(
      Vector To Normalize=&TransformVector(Vector=Linked() {linkedVariable {name=User.SparkDirA, type=NiagaraPosition}}, Destination Space=::Simulation)
      Safe Normalize=true
      Safe Fallback Threshold=0
      Safe Fallback Vector=<0 0 0.001>
     )
     Passthrough Vector:Vector3f=&NormalizeVector(
      Vector To Normalize=&TransformVector(Vector=Linked() {linkedVariable {name=User.SparkDirB, type=NiagaraPosition}}, Destination Space=::Simulation)
      Safe Normalize=true
      Safe Fallback Threshold=0
      Safe Fallback Vector=<0 0 0.001>
     )
    )
    Module.UseConeFalloff=false
    Module.UseVelDistribution=false
    Override Seed=false
    Velocity Mode=::NewEnumerator2
    Velocity Speed=&RandomRangeFloat(Minimum=100, Maximum=3000)
   )
   &ShapeLocation(MaskSpawnGroup=false, Override Seed=false, Sphere Radius=0, UseOwnerScale=true)
   /SetVariables(Particles.Velocity=&Multiply_VectorByFloat(Vector=$Particles.Velocity, Float=$System.FX Scale)) {isSetParameters=true, module=SetVariables_#2}
  ]
  particleUpdateScript [
   &GravityForce(Gravity=&Multiply_VectorByFloat(
     Vector=<0 0 -980>
     Float=&Multiply_Float(A=&RandomRangeFloat(Minimum=0.25, Maximum=1), B=$System.FX Scale)
    ))
   &Drag(Drag=0.35, Use Linear Drag=true, Use Rotational Drag=true)
   &SolveForcesAndVelocity(Clamp Velocity=false, Limit Acceleration=false)
   &ParticleState()
   &ScaleSpriteSize(
    Uniform Curve Scale=$System.FX Scale
    Uniform Curve Sprite Scale=DataInterface() {propertyValues {Curve {keys [<at 0.8 1 RCIM_Cubic>, {arriveTangent=-15.439, interpMode=RCIM_Cubic, leaveTangent=-15.439, tangentMode=RCTM_User, time=1, value=0}]}}}
   )
   &Color(
    Color=&ColorFromCurve(ColorCurve=DataInterface() {propertyValues {
       AlphaCurve {keys [<at 0 1 RCIM_Cubic>,<at 1 0 RCIM_Cubic>]}
       BlueCurve {keys [
         <at 0 37.5>
         {arriveTangentWeight=3.3334, interpMode=RCIM_Cubic, leaveTangentWeight=21.667, tangentMode=RCTM_User, time=0.0737, value=0}
         <at 0.4455 0 RCIM_Cubic>
         <at 0.9911 0>
        ]}
       GreenCurve {keys [
         <at 0 46.875>
         {arriveTangent=-105.2195, interpMode=RCIM_Cubic, leaveTangent=-105.2195, time=0.0737, value=18.75}
         {arriveTangent=-20.4383, interpMode=RCIM_Cubic, leaveTangent=-20.4383, time=0.4455, value=0}
         <at 0.9911 0>
        ]}
       RedCurve {keys [
         <at 0 75>
         {arriveTangent=-228.9853, interpMode=RCIM_Cubic, leaveTangent=-228.9853, tangentMode=RCTM_User, time=0.0737, value=75}
         {arriveTangent=-81.4808, interpMode=RCIM_Cubic, leaveTangent=-81.4808, time=0.4455, value=25}
         <at 0.9911 0.25>
        ]}
      }})
    Scale Alpha=&FloatFromCurve(FloatCurve=DataInterface() {propertyValues {Curve {keys [<at 0 1 RCIM_Cubic>, {arriveTangent=-3.2506, interpMode=RCIM_Cubic, leaveTangent=-3.2506, tangentMode=RCTM_User, time=1, value=0}]}}})
    Scale Color=&VectorFromFloat(Value=0.15)
   )
  ]
  rendererClasses [&NiagaraSpriteRendererProperties]
  renderers [{rendererClass=&NiagaraSpriteRendererProperties}]
 }
 SparksDir001 {
  emitterUpdateScript [
   &EmitterState()
   &SpawnRate(SpawnRate=3000, Use Spawn Probability=false)
  ]
  particleSpawnScript [
   &InitializeParticle(
    Color Mode=::NewEnumerator0
    Lifetime Max=2.5
    Lifetime Mode=::NewEnumerator1
    Sprite Size Max=<5 50>
    Sprite Size Min=<15 25>
    Sprite Size Mode=::NewEnumerator2
    UsePositionOffset=false
    Write Material Random=true
   )
   &PartitionParticles()
   &AddVelocity(
    Module.Use Falloff Distance=false
    Override Seed=false
    Velocity Mode=::NewEnumerator1
    Velocity Speed=&RandomRangeFloat(Minimum=50, Maximum=1200)
   )
   &ShapeLocation(MaskSpawnGroup=false, Override Seed=false, Sphere Radius=0, UseOwnerScale=true)
   /SetVariables(Particles.Velocity=&Multiply_VectorByFloat(Vector=$Particles.Velocity, Float=$System.FX Scale)) {isSetParameters=true, module=SetVariables_#2}
  ]
  particleUpdateScript [
   &GravityForce(Gravity=&Multiply_VectorByFloat(
     Vector=<0 0 -980>
     Float=&Multiply_Float(A=&RandomRangeFloat(Minimum=0.25, Maximum=1), B=$System.FX Scale)
    ))
   &Drag(Drag=0.35, Use Linear Drag=true, Use Rotational Drag=true)
   &SolveForcesAndVelocity(Clamp Velocity=false, Limit Acceleration=false)
   &ParticleState()
   &ScaleSpriteSize(
    Uniform Curve Scale=$System.FX Scale
    Uniform Curve Sprite Scale=DataInterface() {propertyValues {Curve {keys [<at 0.8 1 RCIM_Cubic>, {arriveTangent=-15.439, interpMode=RCIM_Cubic, leaveTangent=-15.439, tangentMode=RCTM_User, time=1, value=0}]}}}
   )
   &Color(
    Color=&ColorFromCurve(ColorCurve=DataInterface() {propertyValues {
       AlphaCurve {keys [<at 0 1 RCIM_Cubic>,<at 1 0 RCIM_Cubic>]}
       BlueCurve {keys [
         <at 0 37.5>
         {arriveTangentWeight=3.3334, interpMode=RCIM_Cubic, leaveTangentWeight=21.667, tangentMode=RCTM_User, time=0.0737, value=0}
         <at 0.4455 0 RCIM_Cubic>
         <at 0.9911 0>
        ]}
       GreenCurve {keys [
         <at 0 46.875>
         {arriveTangent=-105.2195, interpMode=RCIM_Cubic, leaveTangent=-105.2195, time=0.0737, value=18.75}
         {arriveTangent=-20.4383, interpMode=RCIM_Cubic, leaveTangent=-20.4383, time=0.4455, value=0}
         <at 0.9911 0>
        ]}
       RedCurve {keys [
         <at 0 75>
         {arriveTangent=-228.9853, interpMode=RCIM_Cubic, leaveTangent=-228.9853, tangentMode=RCTM_User, time=0.0737, value=75}
         {arriveTangent=-81.4808, interpMode=RCIM_Cubic, leaveTangent=-81.4808, time=0.4455, value=25}
         <at 0.9911 0.25>
        ]}
      }})
    Scale Alpha=&FloatFromCurve(FloatCurve=DataInterface() {propertyValues {Curve {keys [<at 0 1 RCIM_Cubic>, {arriveTangent=-3.2506, interpMode=RCIM_Cubic, leaveTangent=-3.2506, tangentMode=RCTM_User, time=1, value=0}]}}})
    Scale Color=&VectorFromFloat(Value=0.15)
   )
  ]
  rendererClasses [&NiagaraSpriteRendererProperties]
  renderers [{rendererClass=&NiagaraSpriteRendererProperties}]
 }
 SparksDir002 {
  emitterUpdateScript [
   &EmitterState()
   &SpawnRate(SpawnRate=800, Use Spawn Probability=false)
  ]
  particleSpawnScript [
   &InitializeParticle(
    Color Mode=::NewEnumerator0
    Lifetime Max=2.5
    Lifetime Mode=::NewEnumerator1
    Sprite Size Max=<5 50>
    Sprite Size Min=<15 25>
    Sprite Size Mode=::NewEnumerator2
    UsePositionOffset=false
    Write Material Random=true
   )
   &PartitionParticles()
   &AddVelocity(
    Cone Angle=5
    Cone Axis=&self:Mask_Vector_by_Int(
     Int Mask:NiagaraInt32=0
     int:NiagaraInt32=$Particles.PartitionParticles.Partition
     Masked Vector:Vector3f=&NormalizeVector(
      Vector To Normalize=&TransformVector(Vector=Linked() {linkedVariable {name=User.SparkDirA, type=NiagaraPosition}}, Destination Space=::Simulation)
      Safe Normalize=true
      Safe Fallback Threshold=0
      Safe Fallback Vector=<0 0 0.001>
     )
     Passthrough Vector:Vector3f=&NormalizeVector(
      Vector To Normalize=&TransformVector(Vector=Linked() {linkedVariable {name=User.SparkDirB, type=NiagaraPosition}}, Destination Space=::Simulation)
      Safe Normalize=true
      Safe Fallback Threshold=0
      Safe Fallback Vector=<0 0 0.001>
     )
    )
    Module.UseConeFalloff=false
    Module.UseVelDistribution=false
    Override Seed=false
    Velocity Mode=::NewEnumerator2
    Velocity Speed=&RandomRangeFloat(Minimum=100, Maximum=3000)
   )
   &ShapeLocation(MaskSpawnGroup=false, Override Seed=false, Sphere Radius=0, UseOwnerScale=true)
   /SetVariables(Particles.Velocity=&Multiply_VectorByFloat(Vector=$Particles.Velocity, Float=$System.FX Scale)) {isSetParameters=true, module=SetVariables_#2}
  ]
  particleUpdateScript [
   &GravityForce(Gravity=&Multiply_VectorByFloat(
     Vector=<0 0 -980>
     Float=&Multiply_Float(A=&RandomRangeFloat(Minimum=0.25, Maximum=1), B=$System.FX Scale)
    ))
   &Drag(Drag=0.35, Use Linear Drag=true, Use Rotational Drag=true)
   &SolveForcesAndVelocity(Clamp Velocity=false, Limit Acceleration=false)
   &ParticleState()
   &ScaleSpriteSize(
    Uniform Curve Scale=$System.FX Scale
    Uniform Curve Sprite Scale=DataInterface() {propertyValues {Curve {keys [<at 0.8 1 RCIM_Cubic>, {arriveTangent=-15.439, interpMode=RCIM_Cubic, leaveTangent=-15.439, tangentMode=RCTM_User, time=1, value=0}]}}}
   )
   &Color(
    Color=&ColorFromCurve(ColorCurve=DataInterface() {propertyValues {
       AlphaCurve {keys [<at 0 1 RCIM_Cubic>,<at 1 0 RCIM_Cubic>]}
       BlueCurve {keys [
         <at 0 37.5>
         {arriveTangentWeight=3.3334, interpMode=RCIM_Cubic, leaveTangentWeight=21.667, tangentMode=RCTM_User, time=0.0737, value=0}
         <at 0.4455 0 RCIM_Cubic>
         <at 0.9911 0>
        ]}
       GreenCurve {keys [
         <at 0 46.875>
         {arriveTangent=-105.2195, interpMode=RCIM_Cubic, leaveTangent=-105.2195, time=0.0737, value=18.75}
         {arriveTangent=-20.4383, interpMode=RCIM_Cubic, leaveTangent=-20.4383, time=0.4455, value=0}
         <at 0.9911 0>
        ]}
       RedCurve {keys [
         <at 0 75>
         {arriveTangent=-228.9853, interpMode=RCIM_Cubic, leaveTangent=-228.9853, tangentMode=RCTM_User, time=0.0737, value=75}
         {arriveTangent=-81.4808, interpMode=RCIM_Cubic, leaveTangent=-81.4808, time=0.4455, value=25}
         <at 0.9911 0.25>
        ]}
      }})
    Scale Alpha=&FloatFromCurve(FloatCurve=DataInterface() {propertyValues {Curve {keys [<at 0 1 RCIM_Cubic>, {arriveTangent=-3.2506, interpMode=RCIM_Cubic, leaveTangent=-3.2506, tangentMode=RCTM_User, time=1, value=0}]}}})
    Scale Color=&VectorFromFloat(Value=0.15)
   )
  ]
  rendererClasses [&NiagaraSpriteRendererProperties]
  renderers [{rendererClass=&NiagaraSpriteRendererProperties}]
 }
}
summary {
 systemName=NS_GrillMoon_LargeWeaponsWeaponImpact
 emitters [
  {emitterName=SparksDir, simTarget=GPUComputeSim, rendererClasses [&NiagaraSpriteRendererProperties]}
  {
   emitterName=Flash
   rendererClasses [
    &NiagaraSpriteRendererProperties
    &NiagaraLightRendererProperties
    &NiagaraMeshRendererProperties
   ]
  }
  {emitterName=SoftAirflow, rendererClasses [&NiagaraMeshRendererProperties]}
  {emitterName=SparksDir001, rendererClasses [&NiagaraSpriteRendererProperties]}
  {emitterName=SparksDir002, rendererClasses [&NiagaraSpriteRendererProperties]}
 ]
 userVariables [
  User.SparkDirA {type=NiagaraPosition, description="", defaultValue {}}
  User.SparkDirB {type=NiagaraPosition, description="", defaultValue {}}
 ]
}
