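# Niagara impact burst for small weapon hits in GrillMoon kaiju fights, built from three burst emitters scaled by a system FX Scale.
# Flash draws a brief glowing flare through two sprite renderers, with user parameters for size and lifetime.
# SparksDir001 throws speed-scaled sprite sparks under gravity and drag, cooling through an HDR color curve.
# SpikeStreaksMesh fires two staggered bursts of short-lived mesh streaks that lengthen, thin and fade with a panning material.
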NiagaraSystem /GrillMoon/Assets/FX/Kaiju/Fight/General/WeaponImpacts/NS_GrillMoon_SmallWeaponsWeaponImpact
dependencies {
 usedDataInterfaces [
  &NiagaraDataInterfaceCurve
  &NiagaraDataInterfaceVectorCurve
  &NiagaraDataInterfaceVector2DCurve
  &NiagaraDataInterfaceColorCurve
  &NiagaraDataInterfaceDebugDraw
 ]
 usedDynamicInputs [
  &self:Mask_Vector_by_Int_01
  &NormalizeVector
  &TransformVector
 ]
 usedModules [
  &self:SystemSpawnScript.SystemScriptSource.SystemScriptGraph.NiagaraNodeAssignment_4.SetVariables_#1
  &SystemState
  &self:ScratchModule_02
 ]
}
emitters {
 Flash {
  emitterUpdateScript [
   &EmitterState()
   &SpawnBurst_Instantaneous(Use Spawn Probability=false, UseLoopCountLimit=false)
  ]
  particleSpawnScript [
   &InitializeParticle(
    Color=<rgba 1 0.9027 0.639 0.2>
    Lifetime Max=0.35
    Lifetime Min=0.2
    Lifetime Mode=::NewEnumerator1
    Mesh Scale Mode=::NewEnumerator3
    Mesh Uniform Scale=25
    Sprite Rotation Mode=::NewEnumerator2
    Sprite Size Mode=::NewEnumerator3
    Uniform Sprite Size=&Multiply_Float(A=&RandomRangeFloat(Minimum=1.5, Maximum=1), B=1000)
    UsePositionOffset=false
    Write Material Random=true
   )
   /SetVariables(Particles.Lifetime=&Multiply_Float(A=$Particles.Lifetime, B=$User.FlareLifeMul)) {isSetParameters=true}
  ]
  particleUpdateScript [
   &ParticleState()
   &ScaleSpriteSize(
    Uniform Curve Scale=$System.FX Scale
    Uniform Curve Sprite Scale=DataInterface() {propertyValues {Curve {keys [
        {
         arriveTangent=4.7077
         arriveTangentWeight=0.9009
         interpMode=RCIM_Cubic
         leaveTangent=4.7077
         leaveTangentWeight=0.9009
         tangentMode=RCTM_User
         tangentWeightMode=RCTWM_WeightedBoth
         time=0
         value=0
        }
        {
         arriveTangentWeight=0.9145
         interpMode=RCIM_Cubic
         leaveTangentWeight=0.9145
         tangentMode=RCTM_User
         tangentWeightMode=RCTWM_WeightedBoth
         time=1
         value=1
        }
       ]}}}
   )
   &ScaleSpriteSize(Scale Sprite Size Mode=::NewEnumerator0, Uniform Scale Factor=$User.FlareFXSacle) {module=ScaleSpriteSize001}
   &ScaleColor(
    Scale Alpha=&FloatFromCurve(
     FloatCurve=DataInterface() {propertyValues {Curve {keys [
         {arriveTangent=-24.1498, interpMode=RCIM_Cubic, leaveTangent=-24.1498, tangentMode=RCTM_User, time=0, value=5}
         {arriveTangent=-2.1686, interpMode=RCIM_Cubic, leaveTangent=-2.1686, tangentMode=RCTM_User, time=0.4003, value=0.9}
         {interpMode=RCIM_Cubic, leaveTangent=-1.8978, time=1, value=0}
        ]}}}
     Scale Curve=0.2
    )
    Scale RGB=&VectorFromFloat(Value=5)
    ScaleA=true
    ScaleRGB=true
   )
   &FN_Light_Attributes(
    Color=$Particles.Color
    Create Color Attribute=true
    Create Exponent Attribute=true
    Create Radius Attribute=true
    Exponent=1.5
    Radius=&FloatFromCurve(
     FloatCurve=DataInterface() {propertyValues {Curve {keys [
         {arriveTangent=9.0725, interpMode=RCIM_Cubic, leaveTangent=9.0725, tangentMode=RCTM_User, time=0, value=0}
         {arriveTangent=-2.5249, interpMode=RCIM_Cubic, leaveTangent=-2.5249, tangentMode=RCTM_User, time=0.1, value=0.9847}
        ]}}}
     Scale Curve=24472.5293
    )
    Scale Color=6
   )
   &DynamicMaterialParameters(
    Index 0 Param 1=&Lerp_Float(
     A=2
     B=0.5
     Alpha=&FloatFromCurve(FloatCurve=DataInterface() {propertyValues {Curve {keys [
          {
           arriveTangent=4.8139
           arriveTangentWeight=0.5768
           interpMode=RCIM_Cubic
           leaveTangent=4.8139
           leaveTangentWeight=0.5768
           tangentMode=RCTM_User
           tangentWeightMode=RCTWM_WeightedBoth
           time=0
           value=0.1
          }
          {
           arriveTangentWeight=0.9152
           interpMode=RCIM_Cubic
           leaveTangentWeight=0.9152
           tangentMode=RCTM_User
           tangentWeightMode=RCTWM_WeightedBoth
           time=1
           value=1
          }
         ]}}})
    )
    Index 0 Param 2=1
    Index 0 Param 3=1
    Index 0 Param 4=1
    Index 1 Param 1=0.6
    Index 1 Param 2=1
    Index 1 Param 3=1
    Index 1 Param 4=1
    Module.Index0Param0Write=true
    Module.Index0Param1Write=true
    Module.Index0Param2Write=true
    Module.Index0Param3Write=true
    Module.Index1Param0Write=true
    Module.Index1Param1Write=true
    Module.Index1Param2Write=true
    Module.Index1Param3Write=true
    Write Parameter Index 1=true
   )
   /SetVariables(Particles.SPrite 2 dynam0:Vector4f=&MakeVector4(X=1, Y=500, Z=0, W=0)) {isSetParameters=true}
   /SetVariables(Particles.MeshReffract amount:NiagaraFloat=&FloatFromCurve(FloatCurve=DataInterface() {propertyValues {Curve {keys [<at 0.5 1 RCIM_Cubic>,<at 1 0>]}}}, Scale Curve=0.35)) {isSetParameters=true}
  ]
  rendererClasses [&NiagaraSpriteRendererProperties]
  renderers [
   {rendererClass=&NiagaraSpriteRendererProperties}
   {rendererClass=&NiagaraSpriteRendererProperties, rendererIndex=1}
  ]
 }
 SparksDir001 {
  emitterUpdateScript [
   &EmitterState()
   &SpawnBurst_Instantaneous(
    Spawn Count=&RandomRangeInt(Minimum=12, Maximum=17)
    Use Spawn Probability=false
    UseLoopCountLimit=false
   )
  ]
  particleSpawnScript [
   &InitializeParticle(
    Color Mode=::NewEnumerator0
    Lifetime Max=0.7
    Lifetime Mode=::NewEnumerator1
    Sprite Size Max=<3 50>
    Sprite Size Min=<7 25>
    Sprite Size Mode=::NewEnumerator2
    UsePositionOffset=false
    Write Material Random=true
   )
   &PartitionParticles()
   &AddVelocity(
    Module.Use Falloff Distance=false
    Override Seed=false
    Velocity Mode=::NewEnumerator1
    Velocity Speed=&RandomRangeFloat(Minimum=50, Maximum=1200)
   )
   &ShapeLocation(MaskSpawnGroup=false, Override Seed=false, Sphere Radius=0, UseOwnerScale=true)
   /SetVariables(Particles.Velocity=&Multiply_VectorByFloat(Vector=$Particles.Velocity, Float=$System.FX Scale)) {isSetParameters=true}
  ]
  particleUpdateScript [
   &GravityForce(Gravity=&Multiply_VectorByFloat(
     Vector=<0 0 -980>
     Float=&Multiply_Float(A=&RandomRangeFloat(Minimum=0.25, Maximum=1), B=$System.FX Scale)
    ))
   &Drag(Use Linear Drag=true, Use Rotational Drag=true)
   &SolveForcesAndVelocity(Clamp Velocity=false, Limit Acceleration=false)
   &ParticleState()
   &ScaleSpriteSizeBySpeed(Max Scale Factor=<1 1>, Min Scale Factor=<1 0.1>, Velocity Threshold=12000)
   &ScaleSpriteSize(
    Non-Uniform Curve Scale=&Vector2DFromFloat(Value=$System.FX Scale)
    Non-Uniform Curve Sprite Scale=DataInterface() {propertyValues {
      XCurve {keys [
        {arriveTangentWeight=0.2667, interpMode=RCIM_Cubic, leaveTangentWeight=0.0667, tangentMode=RCTM_User, time=0.8, value=1}
        {arriveTangent=-3.2294, arriveTangentWeight=0.0667, interpMode=RCIM_Cubic, leaveTangent=-3.2294, tangentMode=RCTM_User, time=1, value=0.75}
       ]}
      YCurve {keys [
        {arriveTangentWeight=0.2667, interpMode=RCIM_Cubic, leaveTangentWeight=0.0667, tangentMode=RCTM_User, time=0, value=1}
        {arriveTangent=-0.8428, arriveTangentWeight=0.0667, interpMode=RCIM_Cubic, leaveTangent=-0.8428, tangentMode=RCTM_User, time=1, value=0}
       ]}
     }}
    Scale Sprite Size Mode=::NewEnumerator3
   )
   &Color(
    Color=&ColorFromCurve(ColorCurve=DataInterface() {propertyValues {
       AlphaCurve {keys [<at 0 1 RCIM_Cubic>,<at 0.9096 1>,<at 1 0 RCIM_Cubic>]}
       BlueCurve {keys [
         <at 0 5>
         {arriveTangentWeight=3.3334, interpMode=RCIM_Cubic, leaveTangentWeight=21.667, tangentMode=RCTM_User, time=0.3586, value=0}
         <at 0.6681 0 RCIM_Cubic>
         <at 0.9911 0>
        ]}
       GreenCurve {keys [
         <at 0 7>
         {arriveTangent=-4.2409, interpMode=RCIM_Cubic, leaveTangent=-4.2409, time=0.3586, value=30}
         {arriveTangent=-47.4348, interpMode=RCIM_Cubic, leaveTangent=-47.4348, time=0.6681, value=4.1667}
         <at 0.9911 0>
        ]}
       RedCurve {keys [
         <at 0 10>
         {arriveTangent=-228.9853, interpMode=RCIM_Cubic, leaveTangent=-228.9853, tangentMode=RCTM_User, time=0.3586, value=75}
         {arriveTangent=-118.1916, interpMode=RCIM_Cubic, leaveTangent=-118.1916, time=0.6681, value=25}
         <at 0.9911 0.25>
        ]}
      }})
    Scale Alpha=&FloatFromCurve(FloatCurve=DataInterface() {propertyValues {Curve {keys [<at 0 1 RCIM_Cubic>, {arriveTangent=-3.2506, interpMode=RCIM_Cubic, leaveTangent=-3.2506, tangentMode=RCTM_User, time=1, value=0}]}}})
    Scale Color=&VectorFromFloat(Value=0.15)
   )
  ]
  rendererClasses [&NiagaraSpriteRendererProperties]
  renderers [{rendererClass=&NiagaraSpriteRendererProperties}]
 }
 SpikeStreaksMesh {
  emitterUpdateScript [
   &EmitterState()
   &SpawnBurst_Instantaneous(Spawn Count=10, Use Spawn Probability=false)
   &SpawnBurst_Instantaneous(Spawn Count=10, Spawn Time=0.0167, Use Spawn Probability=false, UseLoopCountLimit=false) {module=SpawnBurst_Instantaneous001}
  ]
  particleSpawnScript [
   /Niagara/Modules/Spawn/Initialization/InitializeParticle(
    Lifetime=&UniformRangedFloat(Minimum=0.1, Maximum=0.05)
    Mesh Scale=&Multiply_VectorByFloat(Vector=&UniformRangedVector(Minimum=<0.7 0.4 0.2>, Maximum=<1.35 0.4 0.2>), Float=3)
    Module.Write Color=true
    Module.Write Lifetime=true
    Module.Write Mass=true
    Module.Write Material Random=true
    Module.Write Position=true
    Module.Write Scale=true
    Module.Write SpriteRotation=false
    Module.Write SpriteSize=false
    Position=&Add_Vector(A=/Niagara/DynamicInputs/Helpers/SimulationPosition(), B=<0 0 0>)
   ) {enabled=false}
   /Niagara/Modules/Spawn/Initialization/V2/InitializeParticle(
    Lifetime=&UniformRangedFloat(Minimum=0.1, Maximum=0.05)
    Mesh Scale=&Multiply_VectorByFloat(Vector=&UniformRangedVector(Minimum=<0.7 0.1 0.1>, Maximum=<1.35 0.2 0.2>), Float=3)
    Mesh Scale Mode=::NewEnumerator4
    UsePositionOffset=false
    Write Material Random=true
   ) {module=InitializeParticle001}
   &InitialMeshOrientation(Use Rotation Vector=false)
  ]
  particleUpdateScript [
   &ParticleState()
   &ScaleMeshSize(Scale Factor=&VectorFromFloat(Value=$System.FX Scale)) {module=ScaleMeshSize001}
   &CameraOffset(Camera Offset Amount=10)
   &Color(
    Color=<rgba 3 1.38 0.3 1>
    Scale Alpha=&FloatFromCurve(FloatCurve=DataInterface() {propertyValues {Curve {keys [<at -0.0046 1.0113 RCIM_Cubic>,<at 1.0011 0.002>]}}})
   )
   &ScaleMeshSize(Scale Factor=&VectorFromCurve(VectorCurve=DataInterface() {propertyValues {
       bOptimizeLUT=false
       XCurve {keys [{arriveTangent=1.8635, interpMode=RCIM_Cubic, leaveTangent=1.8635, tangentMode=RCTM_User, time=-0.0048, value=0.4}, <at 0.9737 1>]}
       YCurve {keys [
         {arriveTangent=0.1613, interpMode=RCIM_Cubic, leaveTangent=0.1613, tangentMode=RCTM_User, time=0, value=1}
         {arriveTangent=-1.3776, interpMode=RCIM_Cubic, leaveTangent=-1.3776, tangentMode=RCTM_User, time=0.9944, value=0}
        ]}
       ZCurve {keys [<at 0 1 RCIM_Cubic>, {arriveTangent=-1.635, interpMode=RCIM_Cubic, leaveTangent=-1.635, tangentMode=RCTM_User, time=0.9961, value=0}]}
      }}))
   /Game/Effects/Niagara/Modules/Materials/MaterialPanner(Offset:NiagaraFloat=&Add_Float(A=$Emitter.Age, B=$Particles.Initial.MaterialRandom), Speed=-5)
  ]
  rendererClasses [&NiagaraMeshRendererProperties]
  renderers [{rendererClass=&NiagaraMeshRendererProperties, rendererIndex=0}]
 }
}
summary {
 systemName=NS_GrillMoon_SmallWeaponsWeaponImpact
 emitters [
  {emitterName=Flash, rendererClasses [&NiagaraSpriteRendererProperties]}
  {emitterName=SpikeStreaksMesh, rendererClasses [&NiagaraMeshRendererProperties]}
  {emitterName=SparksDir001, rendererClasses [&NiagaraSpriteRendererProperties]}
 ]
 userVariables [
  User.FlareFXSacle {type=NiagaraFloat, defaultValue=0.65, description=""}
  User.FlareLifeMul {type=NiagaraFloat, defaultValue=0.8, description=""}
  User.SparkDirA {type=NiagaraPosition, description="", defaultValue {}}
  User.SparkDirB {type=NiagaraPosition, description="", defaultValue {}}
 ]
}
